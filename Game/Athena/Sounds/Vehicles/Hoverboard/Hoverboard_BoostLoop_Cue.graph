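# Sound cue for the hoverboard boost loop in the vehicle audio set.
# A mixer layers two looping waves: a hoverboard move loop at reduced volume under a speedy glider thrust loop at full volume.
# It uses the hoverboard movement attenuation and the vehicles sound class, and primes on load.
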SoundCue /Game/Athena/Sounds/Vehicles/Hoverboard/Hoverboard_BoostLoop_Cue
root=Mixer_0
nodes [
 {id=Mixer_0, children [WavePlayer_1,WavePlayer_0], properties {inputVolume [0.25,1]}}
 {id=WavePlayer_0, properties {bLooping=true, soundWaveAssetPtr=/Game/Sounds/Fort_Gadgets/Athena/SFX_Glider_Speedy_Thrust_Loop}}
 {id=WavePlayer_1, properties {bLooping=true, soundWaveAssetPtr=&/Hoverboard_MoveLoop_04}}
]
properties {
 attenuationSettings=/Game/Sounds/Attenuation/Fort_Hoverboard_Movement_Att
 bPrimeOnLoad=true
 maxDistance=10000
 soundClassObject=&Fort_Vehicles
 volumeMultiplier=1
}
