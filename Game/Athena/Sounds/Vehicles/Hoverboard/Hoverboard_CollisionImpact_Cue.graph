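# Sound cue for hoverboard collision impacts, in the vehicles sound class and primed on load.
# A Random node picks one of four collision wave variants with equal weight.
# A Modulator at the root applies variation to the chosen wave, on top of a raised base pitch and reduced volume.
# It is spatialized with the close-range hoverboard movement attenuation.
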SoundCue /Game/Athena/Sounds/Vehicles/Hoverboard/Hoverboard_CollisionImpact_Cue
root=Modulator_0
nodes [
 {id=Modulator_0, children [Random_0]}
 {id=Random_0, children [WavePlayer_0,WavePlayer_1,WavePlayer_2,WavePlayer_3], properties {weights [1,1,1,1]}}
 {id=WavePlayer_0, properties {soundWaveAssetPtr=&/Hoverboard_Collision_04}}
 {id=WavePlayer_1, properties {soundWaveAssetPtr=&/Hoverboard_Collision_01}}
 {id=WavePlayer_2, properties {soundWaveAssetPtr=&/Hoverboard_Collision_02}}
 {id=WavePlayer_3, properties {soundWaveAssetPtr=&/Hoverboard_Collision_03}}
]
properties {
 attenuationSettings=/Game/Sounds/Attenuation/Fort_Hoverboard_Movement_Close_Att
 bPrimeOnLoad=true
 maxDistance=5300
 pitchMultiplier=1.5
 soundClassObject=&Fort_Vehicles
 volumeMultiplier=0.7
}
