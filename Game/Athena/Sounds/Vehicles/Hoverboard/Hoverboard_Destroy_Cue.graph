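# Sound cue for the hoverboard's destruction, playing a single explosion sample in the Athena vehicle audio set.
# One wave player feeds a modulator node at the root, with no randomized choice between samples.
# It uses the medium explosives sound class and borrows the hoverboard movement attenuation asset, and it is set to prime on load.
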SoundCue /Game/Athena/Sounds/Vehicles/Hoverboard/Hoverboard_Destroy_Cue
root=Modulator_0
nodes [
 {id=Modulator_0, children [WavePlayer_1]}
 {id=WavePlayer_1, properties {soundWaveAssetPtr=&/Hoverboard_Explode_01}}
]
properties {
 attenuationSettings=/Game/Sounds/Attenuation/Fort_Hoverboard_Movement_Att
 bPrimeOnLoad=true
 maxDistance=10000
 soundClassObject=/Game/Sounds/FortSoundClassesAndModes/Fort_Explosives_Medium_PSM
 volumeMultiplier=1
}
