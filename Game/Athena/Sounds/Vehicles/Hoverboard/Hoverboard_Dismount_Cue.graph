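# Hoverboard dismount sound cue for vehicle audio in Athena, played at reduced volume.
# A single wave player feeds the dismount sample into a modulator at the root, left at its usual settings.
# The cue uses a close hoverboard movement attenuation asset, belongs to the vehicles sound class and is primed on load.
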SoundCue /Game/Athena/Sounds/Vehicles/Hoverboard/Hoverboard_Dismount_Cue
root=Modulator_0
nodes [
 {id=Modulator_0, children [WavePlayer_3]}
 {id=WavePlayer_3, properties {soundWaveAssetPtr=&/Hoverboard_Dismount_01}}
]
properties {
 attenuationSettings=/Game/Sounds/Attenuation/Fort_Hoverboard_Movement_Close_Att
 bPrimeOnLoad=true
 maxDistance=5300
 soundClassObject=&Fort_Vehicles
 volumeMultiplier=0.7
}
