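# Sound cue for the hoverboard jump sound.
# A mixer layers a random pick from four collision waves at reduced volume with a random pick from two acceleration waves at full volume.
# Only the acceleration layer passes through a modulator for slight random pitch variation.
# The cue uses hoverboard movement attenuation and the vehicle sound class, and is primed on load.
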SoundCue /Game/Athena/Sounds/Vehicles/Hoverboard/Hoverboard_Jump_Cue
root=Mixer_1
nodes [
 {id=Mixer_1, children [Random_0,Modulator_0], properties {inputVolume [0.25,1]}}
 {id=Modulator_0, children [Random_1], properties {pitchMax=1.1, pitchMin=0.9}}
 {id=Random_0, children [WavePlayer_6,WavePlayer_7,WavePlayer_8,WavePlayer_9], properties {weights [1,1,1,1]}}
 {id=Random_1, children [WavePlayer_2,WavePlayer_3], properties {weights [1,1]}}
 {id=WavePlayer_2, properties {soundWaveAssetPtr=&/Hoverboard_Accel_01}}
 {id=WavePlayer_3, properties {soundWaveAssetPtr=&/Hoverboard_Accel_02}}
 {id=WavePlayer_6, properties {soundWaveAssetPtr=&/Hoverboard_Collision_01}}
 {id=WavePlayer_7, properties {soundWaveAssetPtr=&/Hoverboard_Collision_02}}
 {id=WavePlayer_8, properties {soundWaveAssetPtr=&/Hoverboard_Collision_03}}
 {id=WavePlayer_9, properties {soundWaveAssetPtr=&/Hoverboard_Collision_04}}
]
properties {
 attenuationSettings=/Game/Sounds/Attenuation/Fort_Hoverboard_Movement_Att
 bPrimeOnLoad=true
 maxDistance=10000
 soundClassObject=&Fort_Vehicles
 volumeMultiplier=0.7
}
