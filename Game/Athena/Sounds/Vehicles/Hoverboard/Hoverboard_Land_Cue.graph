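# Sound cue for the hoverboard landing sound in Athena, built from two hoverboard dismount waves.
# An equally weighted random node picks one wave, and a modulator varies its pitch slightly on top of a raised base pitch and lowered volume.
# It uses hoverboard movement attenuation and the vehicles sound class, and primes on load.
# A concurrency override allows one instance per owner and blocks new plays while one is active.
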SoundCue /Game/Athena/Sounds/Vehicles/Hoverboard/Hoverboard_Land_Cue
root=Modulator_0
nodes [
 {id=Modulator_0, children [Random_1], properties {pitchMax=1.1, pitchMin=0.9}}
 {id=Random_1, children [WavePlayer_2,WavePlayer_3], properties {weights [1,1]}}
 {id=WavePlayer_2, properties {soundWaveAssetPtr=&/Hoverboard_Dismount_02}}
 {id=WavePlayer_3, properties {soundWaveAssetPtr=&/Hoverboard_Dismount_01}}
]
properties {
 attenuationSettings=/Game/Sounds/Attenuation/Fort_Hoverboard_Movement_Att
 bOverrideConcurrency=true
 bPrimeOnLoad=true
 maxDistance=10000
 pitchMultiplier=1.25
 soundClassObject=&Fort_Vehicles
 volumeMultiplier=0.6
 concurrencyOverrides {bLimitToOwner=true, maxCount=1, resolutionRule=PreventNew}
}
