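# Looping SoundCue for Athena hoverboard movement audio, using the shared hoverboard attenuation and vehicle sound class.
# A mixer blends three hoverboard move loops and a CarbideBlue glider idle loop into a smoothed Doppler node.
# Continuous modulators shape each layer from gameplay parameters: Boost raises pitch on two loops, InAir adds pitch through a chained modulator stack, and Drift and Carve drive layer volumes.
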SoundCue /Game/Athena/Sounds/Vehicles/Hoverboard/Hoverboard_Movement_Cue
root=Doppler_0
nodes [
 {id=Doppler_0, children [Mixer_0], properties {bUseSmoothing=true}}
 {
  id=Mixer_0
  children [ModulatorContinuous_1, ModulatorContinuous_6, ModulatorContinuous_2, ModulatorContinuous_3]
  properties {inputVolume [1,3,0.5,0.8]}
 }
 {
  id=ModulatorContinuous_0
  children [WavePlayer_2]
  properties {
   pitchModulationParams {default=0, maxInput=0.5, maxOutput=0.9, minOutput=0.75}
   volumeModulationParams {default=0, minOutput=0.75, parameterName=Drift}
  }
 }
 {
  id=ModulatorContinuous_1
  children [WavePlayer_5]
  properties {pitchModulationParams {default=0, maxOutput=1.2, minOutput=1, parameterName=Boost}}
 }
 {
  id=ModulatorContinuous_2
  children [WavePlayer_3]
  properties {pitchModulationParams {default=0, maxInput=0.5, maxOutput=1.1, minOutput=0.8}, volumeModulationParams {default=0, parameterName=Drift}}
 }
 {
  id=ModulatorContinuous_3
  children [WavePlayer_4]
  properties {
   pitchModulationParams {default=0, maxInput=0.5, maxOutput=0.9, minOutput=0.75}
   volumeModulationParams {default=0, minInput=0.3, parameterName=Carve}
  }
 }
 {
  id=ModulatorContinuous_4
  children [ModulatorContinuous_0]
  properties {pitchModulationParams {default=0, maxOutput=1.5, minOutput=1, parameterName=Boost}}
 }
 {
  id=ModulatorContinuous_6
  children [ModulatorContinuous_4]
  properties {
   pitchModulationParams {default=0, maxOutput=1.2, minOutput=1, parameterName=InAir}
   volumeModulationParams {default=0, minOutput=1, parameterName=InAir}
  }
 }
 {id=WavePlayer_2, properties {bLooping=true, soundWaveAssetPtr=&/Hoverboard_MoveLoop_08}}
 {id=WavePlayer_3, properties {bLooping=true, soundWaveAssetPtr=/Game/Sounds/Fort_Gadgets/Athena/Glider_CarbideBlue_IdleLoop}}
 {id=WavePlayer_4, properties {bLooping=true, soundWaveAssetPtr=&/Hoverboard_MoveLoop_05}}
 {id=WavePlayer_5, properties {bLooping=true, soundWaveAssetPtr=&/Hoverboard_MoveLoop_01}}
]
properties {
 attenuationSettings=/Game/Sounds/Attenuation/Fort_Hoverboard_Movement_Att
 bPrimeOnLoad=true
 maxDistance=10000
 soundClassObject=&Fort_Vehicles
 volumeMultiplier=0.7
}
