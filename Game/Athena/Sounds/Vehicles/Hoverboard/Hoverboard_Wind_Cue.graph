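# Looping wind sound cue for the hoverboard vehicle, reusing falling-wind umbrella-closed loops.
# A continuous modulator at the root drives volume from a Speed parameter.
# Beneath it, a quality-level node plays one loop for the first two levels and the 1P "opt" loop for the other two.
# The cue primes on load and uses the hoverboard movement attenuation and vehicle sound class.
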SoundCue /Game/Athena/Sounds/Vehicles/Hoverboard/Hoverboard_Wind_Cue
root=ModulatorContinuous_1
nodes [
 {id=ModulatorContinuous_1, children [QualityLevel_0], properties {volumeModulationParams {parameterName=Speed}}}
 {id=QualityLevel_0, children [WavePlayer_2,WavePlayer_2,WavePlayer_0,WavePlayer_0]}
 {id=WavePlayer_0, properties {bLooping=true, soundWaveAssetPtr=/Game/Athena/Sounds/Wind_Falling/Wind_UmbrellaClosed_Loop_1P_opt_02}}
 {id=WavePlayer_2, properties {bLooping=true, soundWaveAssetPtr=/Game/Athena/Sounds/Wind_Falling/Wind_UmbrellaClosed_Loop_01}}
]
properties {
 attenuationSettings=/Game/Sounds/Attenuation/Fort_Hoverboard_Movement_Att
 bPrimeOnLoad=true
 maxDistance=10000
 soundClassObject=&Fort_Vehicles
 volumeMultiplier=1.25
}
