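# Sound cue for the surfboard boost in the hoverboard vehicle sound set, layering hoverboard and surfboard boost samples.
# Two Random nodes each pick one of two equally weighted waves, one hoverboard and one surfboard, and a Mixer plays both picks together at reduced volumes.
# A Modulator at the root varies pitch slightly on each play.
# The cue primes on load and uses the hoverboard boost attenuation and vehicle sound class.
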SoundCue /Game/Athena/Sounds/Vehicles/Hoverboard/Surfboard_Boost_Cue
root=Modulator_0
nodes [
 {id=Mixer_0, children [Random_0,Random_1], properties {inputVolume [0.7,0.8]}}
 {id=Modulator_0, children [Mixer_0], properties {pitchMax=1.1, pitchMin=0.9}}
 {id=Random_0, children [WavePlayer_0,WavePlayer_1], properties {weights [1,1]}}
 {id=Random_1, children [WavePlayer_2,WavePlayer_3], properties {weights [1,1]}}
 {id=WavePlayer_0, properties {soundWaveAssetPtr=&/Hoverboard_Boost_02}}
 {id=WavePlayer_1, properties {soundWaveAssetPtr=&/Hoverboard_Boost_01}}
 {id=WavePlayer_2, properties {soundWaveAssetPtr=&/Surfboard_Boost_03}}
 {id=WavePlayer_3, properties {soundWaveAssetPtr=&/Surfboard_Boost_01}}
]
properties {
 attenuationSettings=/Game/Sounds/Attenuation/Fort_Hoverboard_Boost_Att
 bPrimeOnLoad=true
 maxDistance=10000
 soundClassObject=&Fort_Vehicles
 volumeMultiplier=1
}
