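# Sound cue for the surfboard jump in the hoverboard vehicle sound set, layering a jump one-shot with an acceleration sound.
# A mixer blends a random pick of three surfboard jump waves with a quieter random pick of two hoverboard accel waves, the latter passed through a modulator for slight pitch variation.
# The cue routes to the vehicles sound class, uses hoverboard movement attenuation, and primes on load.
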SoundCue /Game/Athena/Sounds/Vehicles/Hoverboard/Surfboard_Jump_Cue
root=Mixer_1
nodes [
 {id=Mixer_1, children [Random_0,Modulator_0], properties {inputVolume [0.9,0.7]}}
 {id=Modulator_0, children [Random_1], properties {pitchMax=1.1, pitchMin=0.9}}
 {id=Random_0, children [WavePlayer_0,WavePlayer_1,WavePlayer_4], properties {weights [1,1,1]}}
 {id=Random_1, children [WavePlayer_2,WavePlayer_3], properties {weights [1,1]}}
 {id=WavePlayer_0, properties {soundWaveAssetPtr=&/Surfboard_Jump_01}}
 {id=WavePlayer_1, properties {soundWaveAssetPtr=&/Surfboard_Jump_02}}
 {id=WavePlayer_2, properties {soundWaveAssetPtr=&/Hoverboard_Accel_01}}
 {id=WavePlayer_3, properties {soundWaveAssetPtr=&/Hoverboard_Accel_02}}
 {id=WavePlayer_4, properties {soundWaveAssetPtr=&/Surfboard_Jump_03}}
]
properties {
 attenuationSettings=/Game/Sounds/Attenuation/Fort_Hoverboard_Movement_Att
 bPrimeOnLoad=true
 maxDistance=10000
 soundClassObject=&Fort_Vehicles
 volumeMultiplier=0.7
}
