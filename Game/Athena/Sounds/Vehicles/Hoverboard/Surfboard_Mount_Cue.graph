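# Sound cue for mounting the surfboard variant of the hoverboard vehicle.
# A Mixer layers a slightly reduced surfboard mount sound with the hoverboard mount sound, which passes through a Modulator at default settings.
# The cue plays at half volume in the vehicles sound class, uses close-range hoverboard movement attenuation, and primes on load.
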SoundCue /Game/Athena/Sounds/Vehicles/Hoverboard/Surfboard_Mount_Cue
root=Mixer_0
nodes [
 {id=Mixer_0, children [WavePlayer_0,Modulator_0], properties {inputVolume [0.8,1]}}
 {id=Modulator_0, children [WavePlayer_3]}
 {id=WavePlayer_0, properties {soundWaveAssetPtr=&/Surfboard_Mount_01}}
 {id=WavePlayer_3, properties {soundWaveAssetPtr=&/Hoverboard_Mount_01}}
]
properties {
 attenuationSettings=/Game/Sounds/Attenuation/Fort_Hoverboard_Movement_Close_Att
 bPrimeOnLoad=true
 maxDistance=5300
 soundClassObject=&Fort_Vehicles
 volumeMultiplier=0.5
}
